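# Control Rig for the Multibot male character's face accessory part that procedurally drives the head_offset bone from head and left clavicle motion.
# Head rotation relative to spine_05 weights an additive offset, then head_offset takes a 0.71 slerp of the head's local rotation.
# A clamped left clavicle angle weights a second additive shift.
# Debug axes remain enabled on both head bones, while a control-driven head transform branch sits unconnected to execution.
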ControlRigBlueprint /BRCosmetics/Characters/Player/Male/Medium/Bodies/M_MED_Multibot/Meshes/Parts/M_MED_Multibot_FaceAcc_CtrlRig
graphs [
 RigVMFunctionLibrary {links [], nodes []}
 RigVMModel {nodes [
   n1 = &RigUnit_BeginExecution()
   n2 = &RigUnit_GetRelativeBoneTransform(Bone=head, Space=spine_05)
   /Script/ControlRig.RigUnit_ModifyBoneTransforms(
    BoneToModify=((Bone="head_offset",Transform=(Rotation=(X=0,Y=0,Z=0,W=1),Translation=(X=0,Y=4,Z=0),Scale3D=(X=1,Y=1,Z=1))))
    Mode=AdditiveLocal
    WeightMaximum=1
    WeightMinimum=0
    ExecutePin=@n1.ExecutePin
    Weight=&RigVMFunction_MathFloatRemap(
     SourceMaximum=-50
     SourceMinimum=0
     TargetMinimum=0
     bClamp=False
     Value=&RigVMFunction_MathQuaternionToEuler(RotationOrder=XYZ, Value=@n2.Transform.Rotation).Result.Z
    ).Result
   ) {is=struct RigVMUnitNode}
   > /Script/ControlRig.RigUnit_SetBoneRotation(
    Bone=head_offset
    Space=LocalSpace
    Weight=1
    bPropagateToChildren=False
    Rotation=/Script/RigVM.RigVMFunction_MathQuaternionSlerp(A=(X=0,Y=0,Z=0,W=1), T=0.71, B=&RigUnit_GetBoneTransform(Bone=head, Space=LocalSpace).Transform.Rotation) {is=struct RigVMUnitNode, template="Interpolate::Execute(in A,in B,in T,out Result)"}.Result
   ) {is=struct RigVMUnitNode}
   > /Script/ControlRig.RigUnit_DebugTransformMutable(
    Color=(R=1,G=0,B=0,A=1)
    DebugDrawSettings=(DepthPriority=SDPG_Foreground,Lifetime=-1)
    Mode=Axes
    Scale=60
    Space=head
    Thickness=1
    Transform=(Rotation=(X=0,Y=0,Z=0,W=1),Translation=(X=0,Y=0,Z=0),Scale3D=(X=1,Y=1,Z=1))
    WorldOffset=(Rotation=(X=0,Y=0,Z=0,W=1),Translation=(X=0,Y=0,Z=0),Scale3D=(X=1,Y=1,Z=1))
    bEnabled=True
   ) {is=struct RigVMUnitNode}
   > &RigUnit_DebugTransformMutable(
    Color=(R=1,G=0,B=0,A=1)
    DebugDrawSettings=(DepthPriority=SDPG_Foreground,Lifetime=-1)
    Mode=Axes
    Scale=50
    Space=head_offset
    Thickness=0
    Transform=(Rotation=(X=0,Y=0,Z=0,W=1),Translation=(X=0,Y=0,Z=0),Scale3D=(X=1,Y=1,Z=1))
    WorldOffset=(Rotation=(X=0,Y=0,Z=0,W=1),Translation=(X=0,Y=0,Z=0),Scale3D=(X=1,Y=1,Z=1))
    bEnabled=True
   )
   n3 = &RigUnit_GetRelativeBoneTransform(Bone=clavicle_l, Space=spine_05)
   > &RigUnit_ModifyBoneTransforms(
    BoneToModify=((Bone="head_offset",Transform=(Rotation=(X=0,Y=0,Z=0,W=1),Translation=(X=0,Y=0,Z=2.25),Scale3D=(X=1,Y=1,Z=1))))
    Mode=AdditiveLocal
    WeightMaximum=1
    WeightMinimum=0
    Weight=&RigVMFunction_MathFloatRemap(
     SourceMaximum=45
     SourceMinimum=89
     TargetMinimum=0
     bClamp=true
     Value=&RigVMFunction_MathQuaternionToEuler(RotationOrder=ZXY, Value=@n3.Transform.Rotation).Result.Y
    ).Result
   )
   n4 = /Script/ControlRig.RigUnit_Control(
    Base=(Rotation=(X=0,Y=0,Z=0,W=1),Translation=(X=0,Y=0,Z=0),Scale3D=(X=1,Y=1,Z=1))
    Filter=(TranslationFilter=(bX=True,bY=True,bZ=True),RotationFilter=(bX=True,bY=True,bZ=True),ScaleFilter=(bX=True,bY=True,bZ=True))
    InitTransform=/Script/ControlRig.RigUnit_GetInitialBoneTransform(Bone=head, Space=GlobalSpace) {is=struct RigVMUnitNode}.Transform
   ) {is=struct RigVMUnitNode}
   /Script/ControlRig.RigUnit_SetBoneTransform(
    Bone=head
    ExecutePin=()
    Space=LocalSpace
    Weight=1
    bPropagateToChildren=True
    Transform=&RigVMFunction_MathTransformMakeRelative(Global=@n4.Result, Parent=&RigUnit_GetBoneTransform(Bone=neck_02).Transform).Local
   ) {is=struct RigVMUnitNode}
  ]}
]
hierarchy [
 root {loc=<0 0 0>, rot=<0 0 0 1>}
 pelvis {parent=root, loc=<0 -0.7071 93.1191>, rot=<0.0127 0.707 -0.0127 -0.707>}
 spine_01 {parent=pelvis, loc=<6.7494 -0.6388 0>, rot=<0 0 0.034 -0.9994>}
 spine_02 {parent=spine_01, loc=<7.2518 -0.035 0>, rot=<0 0 0.0238 -0.9997>}
 spine_03 {parent=spine_02, loc=<6.0104 0 0>, rot=<0 0 0.0149 -0.9999>}
 spine_04 {parent=spine_03, loc=<8.1024 0 0>, rot=<0 0 -0.0132 -0.9999>}
 spine_05 {parent=spine_04, loc=<10.0146 0 0>, rot=<0 0 -0.0015 -1>}
 clavicle_l {parent=spine_05, loc=<4.3575 -0.1466 -2.3223>, rot=<0.0465 -0.7289 -0.009 -0.6829>}
 clavicle_r {parent=spine_05, loc=<4.3575 -0.1466 2.3223>, rot=<0.7289 0.0465 -0.6829 0.009>}
 neck_01 {parent=spine_05, loc=<11.3138 0.1529 0>, rot=<0 0 -0.2003 -0.9797>}
 neck_02 {parent=neck_01, loc=<4.5627 0 0>, rot=<0 0 0.0504 -0.9987>}
 head {parent=neck_02, loc=<5.3914 0 0>, rot=<0 0 0.115 -0.9934>}
 head_offset {parent=neck_02, loc=<2.6913 1.8435 0>, rot=<0 0 0.115 -0.9934>}
]
